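# Physics asset for the Island Nomad Female pickaxe, covering the dangling feather bones on the axe.
# Six three-segment feather chains of box bodies hang from a small capsule body on dyn_feathers_main.
# Each segment is constrained to the one before it with tight swing and twist limits, and the third chain has slightly tighter twist, keeping feather motion restrained.
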PhysicsAsset /BRCosmetics/Weapons/FORT_Melee/Pickaxe_Island_Nomad_Female/Meshes/Island_Nomad_Female_Axe_PhysicsAsset
bodies {
 dyn_feathers1_01 [{
   shapeType=Box
   extentX=1.7853
   extentY=2.6503
   extentZ=7.6282
   shapeName=dyn_feathers1_01_box
   center=<4.3063 0.1194 -0.0314>
   rotation=<rot 0 -88.0465 -92.1803>
  }]
 dyn_feathers1_02 [{
   shapeType=Box
   extentX=2.2853
   extentY=2.1503
   extentZ=5.1282
   shapeName=dyn_feathers1_01_box
   center=<2.7284 0.1578 0.4913>
   rotation=<rot -0.6877 -93.8163 -96.7746>
  }]
 dyn_feathers1_03 [{
   shapeType=Box
   extentX=2.2853
   extentY=2.6503
   extentZ=7.6282
   shapeName=dyn_feathers1_01_box
   center=<4.2668 -0.6559 0.6717>
   rotation=<rot -1.274 -98.7816 -96.6894>
  }]
 dyn_feathers2_01 [{
   shapeType=Box
   extentX=1.7853
   extentY=4.6503
   extentZ=7.6282
   shapeName=dyn_feathers1_01_box
   center=<5.4495 0.1584 0.0121>
   rotation=<rot -0.1226 -91.2686 -92.1768>
  }]
 dyn_feathers2_02 [{
   shapeType=Box
   extentX=1.7853
   extentY=4.6503
   extentZ=8.6282
   shapeName=dyn_feathers1_01_box
   center=<4.8583 0.0993 0.0677>
   rotation=<rot -0.2032 -93.3932 -94.5445>
  }]
 dyn_feathers2_03 [{
   shapeType=Box
   extentX=1.7853
   extentY=4.6503
   extentZ=7.6282
   shapeName=dyn_feathers1_01_box
   center=<4.0591 -0.7411 0.0575>
   rotation=<rot -0.3034 -96.0435 -92.1589>
  }]
 dyn_feathers3_01 [{
   shapeType=Box
   extentX=0.9582
   extentY=2.7071
   extentZ=7.1357
   shapeName=dyn_feathers3_01_box
   center=<4.248 -0.0394 0.0886>
   rotation=<rot 0 -90.5708 -86.3889>
  }]
 dyn_feathers3_02 [{
   shapeType=Box
   extentX=0.9582
   extentY=2.7071
   extentZ=5.6357
   shapeName=dyn_feathers3_01_box
   center=<3.4752 -0.0319 0.0705>
   rotation=<rot 0 -90.5707 -91.3468>
  }]
 dyn_feathers3_03 [{
   shapeType=Box
   extentX=0.9582
   extentY=2.7071
   extentZ=5.6357
   shapeName=dyn_feathers3_01_box
   center=<3.2447 -0.0296 -0.343>
   rotation=<rot 0 -90.5707 -89.1598>
  }]
 dyn_feathers4_01 [{
   shapeType=Box
   extentX=1.0041
   extentY=2.7122
   extentZ=8.7718
   shapeName=dyn_feathers4_01_box
   center=<4.354 -0.1783 -0.0266>
   rotation=<rot -0.1009 -87.128 -84.9683>
  }]
 dyn_feathers4_02 [{
   shapeType=Box
   extentX=0.9582
   extentY=2.7071
   extentZ=3.6357
   shapeName=dyn_feathers3_01_box
   center=<2.6921 -0.0242 0.2196>
   rotation=<rot 0 -90.5707 -96.6855>
  }]
 dyn_feathers4_03 [{
   shapeType=Box
   extentX=0.9582
   extentY=2.7071
   extentZ=5.6357
   shapeName=dyn_feathers3_01_box
   center=<3.2447 -0.0296 -0.343>
   rotation=<rot 0 -90.5707 -89.1598>
  }]
 dyn_feathers5_01 [{
   shapeType=Box
   extentX=1.0041
   extentY=2.7122
   extentZ=8.7718
   shapeName=dyn_feathers4_01_box
   center=<4.7199 -0.16 -0.022>
   rotation=<rot -0.1009 -87.1276 -89.7675>
  }]
 dyn_feathers5_02 [{
   shapeType=Box
   extentX=0.9582
   extentY=2.7071
   extentZ=8.1357
   shapeName=dyn_feathers3_01_box
   center=<4.7338 -0.0445 0.0862>
   rotation=<rot 0 -90.5707 -89.9489>
  }]
 dyn_feathers5_03 [{
   shapeType=Box
   extentX=1.0462
   extentY=2.4123
   extentZ=8.5481
   shapeName=dyn_feathers5_03_box
   center=<4.7957 -0.6972 -0.3137>
   rotation=<rot 0 -96.3872 -89.5093>
  }]
 dyn_feathers6_01 [{
   shapeType=Box
   extentX=0.9764
   extentY=2.8279
   extentZ=5.9445
   shapeName=dyn_feathers6_01_box
   center=<3.7455 -0.0327 -0.2459>
   rotation=<rot 0 96.2975 -90.0019>
  }]
 dyn_feathers6_02 [{
   shapeType=Box
   extentX=0.9582
   extentY=2.7071
   extentZ=5.6357
   shapeName=dyn_feathers3_01_box
   center=<3.8858 -0.036 0.2039>
   rotation=<rot 0 -90.5707 -95.0442>
  }]
 dyn_feathers6_03 [{
   shapeType=Box
   extentX=0.8694
   extentY=2.7914
   extentZ=5.7703
   shapeName=dyn_feathers6_03_box
   center=<3.1925 -0.4613 0.3404>
   rotation=<rot 0 82.4165 -84.1536>
  }]
 dyn_feathers_main [{shapeType=Capsule, length=0.6615, radius=1.1698}]
}
constraints [
 {bone1Name=dyn_feathers1_01, bone2Name=dyn_feathers_main, swing1LimitDegrees=20, swing2LimitDegrees=20, twistLimitDegrees=5}
 {bone1Name=dyn_feathers1_02, bone2Name=dyn_feathers1_01, swing1LimitDegrees=20, swing2LimitDegrees=20, twistLimitDegrees=5}
 {bone1Name=dyn_feathers1_03, bone2Name=dyn_feathers1_02, swing1LimitDegrees=20, swing2LimitDegrees=20, twistLimitDegrees=5}
 {bone1Name=dyn_feathers2_01, bone2Name=dyn_feathers_main, swing1LimitDegrees=20, swing2LimitDegrees=20, twistLimitDegrees=5}
 {bone1Name=dyn_feathers2_02, bone2Name=dyn_feathers2_01, swing1LimitDegrees=20, swing2LimitDegrees=20, twistLimitDegrees=5}
 {bone1Name=dyn_feathers2_03, bone2Name=dyn_feathers2_02, swing1LimitDegrees=20, swing2LimitDegrees=20, twistLimitDegrees=5}
 {bone1Name=dyn_feathers3_01, bone2Name=dyn_feathers_main, swing1LimitDegrees=20, swing2LimitDegrees=20, twistLimitDegrees=2}
 {bone1Name=dyn_feathers3_02, bone2Name=dyn_feathers3_01, swing1LimitDegrees=20, swing2LimitDegrees=20, twistLimitDegrees=2}
 {bone1Name=dyn_feathers3_03, bone2Name=dyn_feathers3_02, swing1LimitDegrees=20, swing2LimitDegrees=20, twistLimitDegrees=2}
 {bone1Name=dyn_feathers4_01, bone2Name=dyn_feathers_main, swing1LimitDegrees=20, swing2LimitDegrees=20, twistLimitDegrees=5}
 {bone1Name=dyn_feathers4_02, bone2Name=dyn_feathers4_01, swing1LimitDegrees=20, swing2LimitDegrees=20, twistLimitDegrees=5}
 {bone1Name=dyn_feathers4_03, bone2Name=dyn_feathers4_02, swing1LimitDegrees=20, swing2LimitDegrees=20, twistLimitDegrees=5}
 {bone1Name=dyn_feathers5_01, bone2Name=dyn_feathers_main, swing1LimitDegrees=20, swing2LimitDegrees=20, twistLimitDegrees=5}
 {bone1Name=dyn_feathers5_02, bone2Name=dyn_feathers5_01, swing1LimitDegrees=20, swing2LimitDegrees=20, twistLimitDegrees=5}
 {bone1Name=dyn_feathers5_03, bone2Name=dyn_feathers5_02, swing1LimitDegrees=20, swing2LimitDegrees=20, twistLimitDegrees=5}
 {bone1Name=dyn_feathers6_01, bone2Name=dyn_feathers_main, swing1LimitDegrees=20, swing2LimitDegrees=20, twistLimitDegrees=5}
 {bone1Name=dyn_feathers6_02, bone2Name=dyn_feathers6_01, swing1LimitDegrees=20, swing2LimitDegrees=20, twistLimitDegrees=5}
 {bone1Name=dyn_feathers6_03, bone2Name=dyn_feathers6_02, swing1LimitDegrees=20, swing2LimitDegrees=20, twistLimitDegrees=5}
]
