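# Physics asset for the tassel on the Island Nomad Female pickaxe, a Battle Royale cosmetic melee weapon.
# Four capsule bodies cover a short, wide feather root bone and three thin tassel segments.
# Constraints link them in a serial chain from dyn_feathers_main through dyn_tassel03.
# Swing and twist limits keep twist tight at every joint, with a wider first swing axis where the tassel meets the feathers.
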PhysicsAsset /BRCosmetics/Weapons/FORT_Melee/Pickaxe_Island_Nomad_Female/Meshes/Island_Nomad_Female_Axe_Tassel_PhysicsAsset
bodies {
 dyn_feathers_main [{shapeType=Capsule, length=0.6615, radius=1.1698}]
 dyn_tassel01 [{shapeType=Capsule, length=5, radius=0.505, shapeName=dyn_tassel01_capsule, center=<3.9328 -0.0001 0>, rotation=<rot -90 -1.8984 1.8984>}]
 dyn_tassel02 [{shapeType=Capsule, length=5, radius=0.505, shapeName=dyn_tassel01_capsule, center=<3.5245 -0.0001 0.4032>, rotation=<rot -84.5399 0 0>}]
 dyn_tassel03 [{
   shapeType=Capsule
   length=5
   radius=0.505
   shapeName=dyn_tassel01_capsule
   center=<2.4686 -0.0001 -0.0033>
   rotation=<rot -89.7992 179.9999 180>
  }]
}
constraints [
 {bone1Name=dyn_tassel01, bone2Name=dyn_feathers_main, swing1LimitDegrees=45, swing2LimitDegrees=25, twistLimitDegrees=5}
 {bone1Name=dyn_tassel02, bone2Name=dyn_tassel01, swing1LimitDegrees=25, swing2LimitDegrees=25, twistLimitDegrees=5}
 {bone1Name=dyn_tassel03, bone2Name=dyn_tassel02, swing1LimitDegrees=25, swing2LimitDegrees=25, twistLimitDegrees=5}
]
