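# Animation Blueprint for the Island Nomad Female pickaxe that adds secondary motion to its dangling tassel and feathers.
# In front end, chained RigidBody nodes use FE physics assets, falling back to Trail controllers with additive tassel rotations when bRigidBodyEnabled is false.
# Otherwise, RigidBody nodes use standard physics assets, followed by a tassel bone offset.
# All branches share a cached input pose; simulations run in base-bone space with custom gravity.
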AnimBlueprint /BRCosmetics/Weapons/FORT_Melee/Pickaxe_Island_Nomad_Female/Meshes/Island_Nomad_Female_Axe_Skeleton_AnimBP
target_skeleton=&/Island_Nomad_Female_Axe_Skeleton
graphs [
 AnimGraph {nodes [
   n1 = RigidBody(ComponentPose=Local To Component(LocalPose=Use cached pose 'PreSim'())) {
    bClampLinearTranslationLimitToRefPose=true
    bOverrideWorldGravity=true
    overridePhysicsAsset=&/Island_Nomad_Female_Axe_PhysicsAsset_FE
    simulationSpace=BaseBoneSpace
    alphaBoolBlend {blendOption=Linear}
    baseBoneRef {boneName=b_AR_Root}
    componentAppliedLinearAccClamp=<3000 3000 3000>
    componentLinearAccScale=<0.4 0.3 0.3>
    componentLinearVelScale=<0.4 0.3 0.3>
    overrideWorldGravity {z=-300}
   }
   n2 = RigidBody(ComponentPose=@n1) {
    bClampLinearTranslationLimitToRefPose=true
    bOverrideWorldGravity=true
    overridePhysicsAsset=&/Island_Nomad_Female_Axe_Tassel_PhysicsAsset_FE
    simulationSpace=BaseBoneSpace
    alphaBoolBlend {blendOption=Linear}
    baseBoneRef {boneName=b_AR_Root}
    componentAppliedLinearAccClamp=<3000 3000 3000>
    componentLinearAccScale=<0.4 0.3 0.3>
    componentLinearVelScale=<0.4 0.3 0.3>
    overrideWorldGravity {z=-700}
   }
   n3 = Component To Local(ComponentPose=@n2)
   n4 = Trail controller - Bone: dyn_tassel02(ComponentPose=Local To Component(LocalPose=Use cached pose 'PreSim'()), Alpha=1) {
    relaxationSpeedScale=4.5
    baseJoint {boneName=dyn_tassel01}
    rotationLimits [{limitMax=<180 180 180>}, {limitMax=<180 180 180>}]
    rotationOffsets [<0 0 0>, <0 0 0>]
    trailBone {boneName=dyn_tassel02}
   }
   n5 = Trail controller - Bone: dyn_feathers1_03(ComponentPose=@n4, Alpha=1) {
    chainLength=3
    relaxationSpeedScale=3.5
    baseJoint {boneName=dyn_feathers1_01}
    rotationLimits [{limitMax=<180 180 180>}, {limitMax=<180 180 180>}, {limitMax=<180 180 180>}]
    rotationOffsets [<0 0 0>, <0 0 0>, <0 0 0>]
    trailBone {boneName=dyn_feathers1_03}
   }
   n6 = Trail controller - Bone: dyn_feathers2_03(ComponentPose=@n5, Alpha=1) {
    chainLength=3
    relaxationSpeedScale=3.5
    baseJoint {boneName=dyn_feathers2_01}
    rotationLimits [{limitMax=<180 180 180>}, {limitMax=<180 180 180>}, {limitMax=<180 180 180>}]
    rotationOffsets [<0 0 0>, <0 0 0>, <0 0 0>]
    trailBone {boneName=dyn_feathers2_03}
   }
   n7 = Trail controller - Bone: dyn_feathers3_03(ComponentPose=@n6, Alpha=1) {
    chainLength=3
    relaxationSpeedScale=3.5
    baseJoint {boneName=dyn_feathers3_01}
    rotationLimits [{limitMax=<180 180 180>}, {limitMax=<180 180 180>}, {limitMax=<180 180 180>}]
    rotationOffsets [<0 0 0>, <0 0 0>, <0 0 0>]
    trailBone {boneName=dyn_feathers3_03}
   }
   n8 = Trail controller - Bone: dyn_feathers4_03(ComponentPose=@n7, Alpha=1) {
    chainLength=3
    relaxationSpeedScale=3.5
    baseJoint {boneName=dyn_feathers4_01}
    rotationLimits [{limitMax=<180 180 180>}, {limitMax=<180 180 180>}, {limitMax=<180 180 180>}]
    rotationOffsets [<0 0 0>, <0 0 0>, <0 0 0>]
    trailBone {boneName=dyn_feathers4_03}
   }
   n9 = Trail controller - Bone: dyn_feathers5_03(ComponentPose=@n8, Alpha=1) {
    chainLength=3
    relaxationSpeedScale=3.5
    baseJoint {boneName=dyn_feathers5_01}
    rotationLimits [{limitMax=<180 180 180>}, {limitMax=<180 180 180>}, {limitMax=<180 180 180>}]
    rotationOffsets [<0 0 0>, <0 0 0>, <0 0 0>]
    trailBone {boneName=dyn_feathers5_03}
   }
   n10 = Trail controller - Bone: dyn_feathers6_03(ComponentPose=@n9, Alpha=1) {
    chainLength=3
    relaxationSpeedScale=3.5
    baseJoint {boneName=dyn_feathers6_01}
    rotationLimits [{limitMax=<180 180 180>}, {limitMax=<180 180 180>}, {limitMax=<180 180 180>}]
    rotationOffsets [<0 0 0>, <0 0 0>, <0 0 0>]
    trailBone {boneName=dyn_feathers6_03}
   }
   n11 = "Transform (Modify) Bone - Bone: dyn_tassel01"(ComponentPose=@n10, Alpha=1, Translation="0,0,0", Rotation="-20,16,0", Scale="1,1,1") {
    rotationMode=BMM_Additive
    rotationSpace=BCS_BoneSpace
    alphaBoolBlend {blendOption=Linear}
    boneToModify {boneName=dyn_tassel01}
    rotation {pitch=-20, yaw=16}
   }
   n12 = "Transform (Modify) Bone - Bone: dyn_tassel02"(ComponentPose=@n11, Alpha=1, Translation="0,0,0", Rotation="-20,0,0", Scale="1,1,1") {
    rotationMode=BMM_Additive
    rotationSpace=BCS_BoneSpace
    alphaBoolBlend {blendOption=Linear}
    boneToModify {boneName=dyn_tassel02}
    rotation {pitch=-20}
   }
   n13 = "Transform (Modify) Bone - Bone: dyn_tassel03"(ComponentPose=@n12, Alpha=1, Translation="0,0,0", Rotation="-20,0,0", Scale="1,1,1") {
    rotationMode=BMM_Additive
    rotationSpace=BCS_BoneSpace
    alphaBoolBlend {blendOption=Linear}
    boneToModify {boneName=dyn_tassel03}
    rotation {pitch=-20}
   }
   n14 = Component To Local(ComponentPose=@n13)
   n15 = Get bRigidBodyEnabled()
   n16 = RigidBody(ComponentPose=Local To Component(LocalPose=Use cached pose 'PreSim'())) {
    bClampLinearTranslationLimitToRefPose=true
    bOverrideWorldGravity=true
    overridePhysicsAsset=&/Island_Nomad_Female_Axe_PhysicsAsset
    simulationSpace=BaseBoneSpace
    alphaBoolBlend {blendOption=Linear}
    baseBoneRef {boneName=b_AR_Root}
    componentAppliedLinearAccClamp=<3000 3000 3000>
    componentLinearAccScale=<0.4 0.3 0.3>
    componentLinearVelScale=<0.4 0.3 0.3>
    overrideWorldGravity {z=-300}
   }
   n17 = RigidBody(ComponentPose=@n16) {
    bClampLinearTranslationLimitToRefPose=true
    bOverrideWorldGravity=true
    overridePhysicsAsset=&/Island_Nomad_Female_Axe_Tassel_PhysicsAsset
    simulationSpace=BaseBoneSpace
    alphaBoolBlend {blendOption=Linear}
    baseBoneRef {boneName=b_AR_Root}
    componentAppliedLinearAccClamp=<3000 3000 3000>
    componentLinearAccScale=<0.4 0.3 0.3>
    componentLinearVelScale=<0.4 0.3 0.3>
    overrideWorldGravity {z=-700}
   }
   n18 = "Transform (Modify) Bone - Bone: dyn_tassel01"(ComponentPose=@n17, Alpha=1, Translation="0,0,0", Rotation="15,-10,0", Scale="1,1,1") {
    alpha=0
    rotationMode=BMM_Additive
    rotationSpace=BCS_BoneSpace
    alphaBoolBlend {blendOption=Linear}
    boneToModify {boneName=dyn_tassel01}
    rotation {pitch=15, yaw=-10}
   }
   Output Pose(Result=Blend Poses by bool(
     BlendPose_0=Blend Poses by bool(BlendPose_0=@n3, BlendPose_1=@n14, bActiveValue=@n15) {blendType=Linear}
     BlendPose_1=Component To Local(ComponentPose=@n18)
     bActiveValue=Get bIsInFrontEnd()
    ) {blendType=Linear}) {layerGroup=None}
   Save cached pose 'PreSim'(Pose=Input Pose()) {cacheName=PreSim}
  ]}
 EventGraph {nodes [
   TryGetPawnOwner()
   Event BlueprintUpdateAnimation()
  ]}
]
variables []
